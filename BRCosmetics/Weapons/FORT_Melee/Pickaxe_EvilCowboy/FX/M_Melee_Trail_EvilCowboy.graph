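# Translucent, two-sided flame trail material for the Evil Cowboy pickaxe swing, built for beam trails and tinted by particle color.
# A panning normal map, scaled by the UVDistortion dynamic parameter, warps two scrolling cartoon-flame layers and sine-offset fire masks.
# Their product is shaped by sphere masks and UV gradients, then faded by near-camera distance, fresnel and particle alpha.
# Depth fade softens opacity on the default shading path, while mobile skips it.
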Material /BRCosmetics/Weapons/FORT_Melee/Pickaxe_EvilCowboy/FX/M_Melee_Trail_EvilCowboy
graph {
 properties {
  bUsedWithBeamTrails=true
  blendMode=BLEND_Translucent
  pixelDepthOffsetMode=PDOM_Legacy
  translucencyLightingMode=TLM_VolumetricPerVertexDirectional
  twoSided=true
  usedShadingModels=Default Lit
 }
 nodes [
  n1 = ParticleColor()
  n2 = Multiply(A=@n1.RGB) {constB=0.25}
  n3 = ComponentMask(None=TextureCoordinate() {coordinateIndex=1}) {r=true}
  MaterialFunctionCall(Emissive=Multiply(A=LinearInterpolate(A=@n1.RGB, B=@n2, Alpha=@n3), B=@n1.RGB)) {materialFunction=&MF_FX_EmissiveAdjustments}
  n4 = TextureCoordinate() {coordinateIndex=1, uTiling=1.3}
  n5 = ScalarParameter() {defaultValue=2, parameterName=UScale}
  n6 = ScalarParameter() {defaultValue=1, parameterName=VScale}
  n7 = TextureSample(UVs=Panner(Coordinate=TextureCoordinate() {coordinateIndex=1}) {speedX=1, speedY=0.1}) {
   samplerSource=SSM_Wrap_WorldGroupSettings
   samplerType=SAMPLERTYPE_Normal
   texture=/Game/Effects/Fort_Effects/Gameplay/SmokeDustExplosions/Textures/T_Explosion_NM_01
  }
  n8 = ComponentMask(None=MaterialFunctionCall(Normal=@n7.RGB, Flatness=DynamicParameter() {defaultValue=<rgba 0.975 1 1 1>, paramNames [UVDistortion,Param2,Param3,Param4]}.UVDistortion) {materialFunction=&FlattenNormal}.Result) {g=true, r=true}
  n9 = Add(A=Multiply(A=TextureCoordinate() {coordinateIndex=1}, B=AppendVector(A=@n5, B=@n6)), B=@n8)
  n10 = ScalarParameter() {defaultValue=2, parameterName=Bias}
  n11 = TextureSample!(UVs=Panner(Coordinate=Add(A=@n9) {constB=0.37}) {speedX=-0.8, speedY=-0.8}, Tex, Bias=@n10, Apply View MipBias) {
   mipValueMode=TMVM_MipBias
   samplerSource=SSM_Wrap_WorldGroupSettings
   texture=/Game/Effects/Fort_Effects/Textures/Tile/T_Tile_Texture_Cartoon_Flames_Layer
  }
  n12 = TextureSample(UVs=Panner(Coordinate=@n9) {speedX=-0.6, speedY=-0.6}, Bias=@n10) {
   mipValueMode=TMVM_MipBias
   samplerSource=SSM_Wrap_WorldGroupSettings
   texture=&T_Tile_Texture_Cartoon_Flames_Layer
  }
  n13 = ComponentMask(None=TextureCoordinate()) {g=true}
  n14 = Add(A=Add(A=Multiply(A=@n11.R) {constB=0.5}, B=Multiply(A=@n12.R) {constB=0.5}), B=Power(Base=OneMinus(None=@n13)) {constExponent=10})
  n15 = Add(A=TextureCoordinate() {coordinateIndex=1}, B=@n8)
  n16 = OneMinus(None=ComponentMask(None=@n15) {g=true})
  n17 = OneMinus(None=ComponentMask(None=@n15) {g=true})
  n18 = TextureCoordinate()
  n19 = LinearInterpolate(
   A=AppendVector(A=ComponentMask(None=@n15) {r=true}, B=Add(A=@n16) {constB=-0.15})
   B=AppendVector(A=ComponentMask(None=@n15) {r=true}, B=Add(A=@n17) {constB=0.45})
   Alpha=Sine(None=ComponentMask(None=@n18) {r=true}) {period=2}
  )
  n20 = ScalarParameter() {defaultValue=1.5, parameterName=Bias_A}
  n21 = Add(A=TextureCoordinate() {coordinateIndex=1}, B=@n8)
  n22 = OneMinus(None=ComponentMask(None=@n21) {g=true})
  n23 = OneMinus(None=ComponentMask(None=@n21) {g=true})
  n24 = TextureCoordinate()
  n25 = LinearInterpolate(
   A=AppendVector(A=ComponentMask(None=@n21) {r=true}, B=Add(A=@n22) {constB=-0.15})
   B=AppendVector(A=ComponentMask(None=@n21) {r=true}, B=Add(A=@n23) {constB=0.45})
   Alpha=Sine(None=ComponentMask(None=@n24) {r=true}) {period=2}
  )
  n26 = Multiply(
   A=TextureSample(UVs=Panner(Coordinate=@n19) {speedX=-1}, Bias=@n20) {mipValueMode=TMVM_MipBias, texture=/Game/Effects/Fort_Effects/Textures/Fire/T_Fire_Mask}.R
   B=TextureSample(UVs=Panner(Coordinate=@n25) {speedX=-1}, Bias=@n20) {mipValueMode=TMVM_MipBias, texture=&T_Fire_Mask}.R
  )
  n27 = Constant() {r=0.5}
  n28 = MaterialFunctionCall(Float2=TextureCoordinate()) {materialFunction=&BreakOutFloat2Components}
  n29 = Multiply(
   A=Clamp(None=Multiply(A=@n14, B=@n26))
   B=SphereMask(
    A=AppendVector(A=@n27, B=@n28.G)
    B=@n27
    Radius=ScalarParameter() {defaultValue=0.5, parameterName=Mask Radius}
    Hardness=ScalarParameter() {defaultValue=0.7, parameterName=Mask Hardness}
   )
  )
  n30 = TextureCoordinate() {uTiling=0.5}
  n31 = Constant2Vector() {g=0.82, r=0.85}
  n32 = Constant() {r=0.8}
  n33 = Constant() {r=0.3}
  n34 = Power(Base=OneMinus(None=SphereMask(A=@n30, B=@n31, Radius=@n32, Hardness=@n33))) {constExponent=5}
  n35 = Multiply(A=ComponentMask(None=TextureCoordinate()) {r=true}) {constB=3}
  n36 = Multiply(
   A=Multiply(A=Multiply(A=@n29, B=@n34), B=Saturate(None=@n35))
   B=MaterialFunctionCall(Distance=ScalarParameter() {defaultValue=64, parameterName=NearFadeDistance}) {materialFunction=&Near_Camera_Fade}.Result
  )
  n37 = ScalarParameter() {defaultValue=5, parameterName=Fresnel Brightness}
  n38 = ScalarParameter() {defaultValue=1, parameterName=Fresnel Exponent}
  n39 = VertexNormalWS()
  n40 = Clamp(None=Subtract(
    A=MaterialFunctionCall(Brightness=@n37, Exponent=@n38, Vertex / Pixel Normal=@n39) {materialFunction=&MF_Fresnel_FX}.Fresnel Double Sided
    B=ScalarParameter() {defaultValue=0.1, parameterName=Frensel_subtract}
   ))
  n41 = Multiply(A=Power(Base=ComponentMask(None=@n4) {r=true}), B=Multiply(A=@n1.A, B=Multiply(A=@n36, B=@n40)))
  ShadingPathSwitch(Default=DepthFade(Opacity=@n41, FadeDistance=ScalarParameter() {defaultValue=250, parameterName=FadeDistance}), Mobile=@n41) {desc=Migrated from FeatureLevelSwitch}
 ]
}
