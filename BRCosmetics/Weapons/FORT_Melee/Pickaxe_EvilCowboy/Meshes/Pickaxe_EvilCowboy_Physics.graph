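# Physics asset for the Evil Cowboy two-handed pickaxe, defining bodies and joints for its dangling flaps and low chain.
# Capsule flap chains and a sphere chain, locked at its root, branch from the handle with mostly limited swing and twist.
# A free-rotating mid-flap base carries left, right and middle flaps, while the base-flap chain forks at its tip.
# The middle flap's side branch is locked back to the main strip, closing a loop.
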PhysicsAsset /BRCosmetics/Weapons/FORT_Melee/Pickaxe_EvilCowboy/Meshes/Pickaxe_EvilCowboy_Physics
bodies {
 Melee_TwoHanded_Handle [{shapeType=Capsule, length=99.7254, radius=2.1024, center=<0.8374 -1.5397 24.8335>, rotation=<rot 1.395 -90 -0.0527>}]
 dyn_BaseFlap01 [{shapeType=Capsule, length=5.8705, radius=1.2006, center=<3.3646 -0.0001 0>, rotation=<rot 90 0 0>}]
 dyn_BaseFlap02 [{shapeType=Capsule, length=6.3692, radius=1.1629, center=<3.3646 -0.0001 0>, rotation=<rot 90 0 0>}]
 dyn_BaseFlap03 [{shapeType=Capsule, length=5.974, radius=1.7714, center=<3.3646 -0.0001 0>, rotation=<rot 90 0 0>}]
 dyn_BaseFlap04 [{shapeType=Capsule, length=3.5278, radius=2.6852, center=<2.6425 0 0>, rotation=<rot 90 0 0>}]
 dyn_BaseFlap05_l [{shapeType=Capsule, length=3.9599, radius=1.7003, center=<1.5035 0 0>, rotation=<rot 90 0 0>}]
 dyn_BaseFlap05_r [{shapeType=Capsule, length=2.8383, radius=0.8709, center=<1.5035 0 0>, rotation=<rot 90 0 0>}]
 dyn_flapMid1 [{shapeType=Capsule, length=0.625, radius=2.005, center=<2.1156 0 0>, rotation=<rot 90 0 0>}]
 dyn_flapMid2 [{shapeType=Capsule, length=1.4463, radius=2.75, center=<2.7373 0 0>, rotation=<rot 90 0 0>}]
 dyn_flapMid3 [{shapeType=Capsule, length=2.4687, radius=1.8988, center=<2.2792 -0.1359 0>, rotation=<rot 90 161.0296 163.0237>}]
 dyn_flapMid4 [{shapeType=Capsule, length=3.0512, radius=1.8347, center=<3.5239 -0.0001 0>, rotation=<rot 90 0 0>}]
 dyn_flapMid5 [{shapeType=Capsule, length=3.1211, radius=1.678, center=<3.5654 -0.5846 0.4658>, rotation=<rot 0 80.6878 -82.6538>}]
 dyn_flapMid_r_1 [{
   shapeType=Capsule
   length=1.9302
   radius=1.375
   shapeName=dyn_flapMid_r_1_capsule
   center=<1.8508 -0.1492 0.0525>
   rotation=<rot 0.5281 -84.5796 -83.5854>
  }]
 dyn_flapMid_r_2 [{
   shapeType=Capsule
   length=2.6669
   radius=1.2909
   shapeName=dyn_flapMid_r_2_capsule
   center=<1.8824 0.4821 -0.3418>
   rotation=<rot 1.2396 -96.8039 -67.0535>
  }]
 dyn_flapMid_r_3 [{
   shapeType=Capsule
   length=3.3151
   radius=0.7647
   shapeName=dyn_flapMid_r_3_capsule
   center=<0.9365 -0.1047 -1.7837>
   rotation=<rot 13.9938 -90 -30.6229>
  }]
 dyn_flap_l_1 [{shapeType=Capsule, length=5.3008, radius=1.9595, center=<3.4901 -0.1388 0.0724>, rotation=<rot 0 87.725 -88.813>}]
 dyn_flap_l_2 [{shapeType=Capsule, length=6.9489, radius=1.1216, center=<2.6797 -0.0197 0.0793>, rotation=<rot 0 89.5809 -88.3044>}]
 dyn_flap_l_3 [{shapeType=Capsule, length=3.8451, radius=0.75, center=<1.0313 0.0819 0.0134>, rotation=<rot 0 94.5491 -89.2558>}]
 dyn_flap_r_01 [{shapeType=Capsule, length=5.6393, radius=1.625, center=<3.5025 0.7632 -0.4338>, rotation=<rot 2.5956 -89.9995 88.3609>}]
 dyn_flap_r_02 [{shapeType=Capsule, length=5.0885, radius=0.625, center=<2.8325 0.9172 0.4729>, rotation=<rot 5.8842 -89.9986 93.7662>}]
 dyn_flap_r_03 [{shapeType=Capsule, length=2.4374, radius=0.8419, center=<0.5275 0.785 0.3184>, rotation=<rot -0.4195 132.9626 -89.4828>}]
 dyn_lowChain01 [{shapeType=Sphere, radius=4, center=<-3.9671 0.6168 0.0001>}]
 dyn_lowChain02 [{shapeType=Sphere, radius=1.5}]
 dyn_lowChain03 [{shapeType=Sphere, radius=1.5}]
 dyn_lowChain04 [{shapeType=Sphere, radius=1.5}]
 dyn_lowChain05 [{shapeType=Sphere, radius=1.5}]
 dyn_midFlapBase [{shapeType=Capsule, length=19.0241, radius=5.4978, center=<1.8389 0.8917 0.3135>, rotation=<rot 0.5143 -90.0035 0.0861>}]
}
constraints [
 {
  bone1Name=dyn_lowChain01
  bone2Name=Melee_TwoHanded_Handle
  swing1LimitDegrees=45
  swing1Motion=Locked
  swing2LimitDegrees=45
  swing2Motion=Locked
  twistLimitDegrees=45
  twistMotion=Locked
 }
 {bone1Name=dyn_lowChain02, bone2Name=dyn_lowChain01, swing1LimitDegrees=45, swing2LimitDegrees=20, twistLimitDegrees=0.01}
 {bone1Name=dyn_lowChain03, bone2Name=dyn_lowChain02, swing1LimitDegrees=45, swing2LimitDegrees=45, twistLimitDegrees=0.01}
 {bone1Name=dyn_lowChain04, bone2Name=dyn_lowChain03, swing1LimitDegrees=45, swing2LimitDegrees=45, twistLimitDegrees=0.01}
 {bone1Name=dyn_lowChain05, bone2Name=dyn_lowChain04, swing1LimitDegrees=45, swing2LimitDegrees=45, twistLimitDegrees=0.01}
 {bone1Name=dyn_BaseFlap01, bone2Name=Melee_TwoHanded_Handle, swing1LimitDegrees=20, swing2LimitDegrees=20, twistLimitDegrees=10}
 {bone1Name=dyn_BaseFlap02, bone2Name=dyn_BaseFlap01, swing1LimitDegrees=20, swing2LimitDegrees=20, twistLimitDegrees=5}
 {bone1Name=dyn_BaseFlap03, bone2Name=dyn_BaseFlap02, swing1LimitDegrees=30, swing2LimitDegrees=30, twistLimitDegrees=10}
 {bone1Name=dyn_BaseFlap04, bone2Name=dyn_BaseFlap03, swing1LimitDegrees=45, swing2LimitDegrees=45, twistLimitDegrees=10}
 {bone1Name=dyn_flap_l_1, bone2Name=dyn_midFlapBase, swing1LimitDegrees=45, swing2LimitDegrees=45, twistLimitDegrees=10}
 {bone1Name=dyn_flap_l_2, bone2Name=dyn_flap_l_1, swing1LimitDegrees=45, swing2LimitDegrees=45, twistLimitDegrees=5}
 {bone1Name=dyn_flap_l_3, bone2Name=dyn_flap_l_2, swing1LimitDegrees=45, swing2LimitDegrees=45, twistLimitDegrees=5}
 {bone1Name=dyn_flap_r_01, bone2Name=dyn_midFlapBase, swing1LimitDegrees=45, swing2LimitDegrees=45, twistLimitDegrees=10}
 {bone1Name=dyn_flap_r_02, bone2Name=dyn_flap_r_01, swing1LimitDegrees=45, swing2LimitDegrees=45, twistLimitDegrees=5}
 {bone1Name=dyn_flap_r_03, bone2Name=dyn_flap_r_02, swing1LimitDegrees=45, swing2LimitDegrees=45, twistLimitDegrees=5}
 {
  bone1Name=dyn_midFlapBase
  bone2Name=Melee_TwoHanded_Handle
  swing1LimitDegrees=45
  swing1Motion=Free
  swing2LimitDegrees=45
  swing2Motion=Free
  twistLimitDegrees=45
  twistMotion=Free
 }
 {bone1Name=dyn_flapMid1, bone2Name=dyn_midFlapBase, swing1LimitDegrees=10, swing2LimitDegrees=10, twistLimitDegrees=0}
 {bone1Name=dyn_flapMid2, bone2Name=dyn_flapMid1, swing1LimitDegrees=20, swing2LimitDegrees=20, twistLimitDegrees=10}
 {bone1Name=dyn_flapMid3, bone2Name=dyn_flapMid2, swing1LimitDegrees=20, swing2LimitDegrees=20, twistLimitDegrees=10}
 {bone1Name=dyn_flapMid4, bone2Name=dyn_flapMid3, swing1LimitDegrees=20, swing2LimitDegrees=20, twistLimitDegrees=10}
 {bone1Name=dyn_flapMid5, bone2Name=dyn_flapMid4, swing1LimitDegrees=20, swing2LimitDegrees=20, twistLimitDegrees=10}
 {bone1Name=dyn_flapMid_r_1, bone2Name=dyn_flapMid1, swing1LimitDegrees=20, swing2LimitDegrees=20, twistLimitDegrees=10}
 {bone1Name=dyn_flapMid_r_2, bone2Name=dyn_flapMid_r_1, swing1LimitDegrees=20, swing2LimitDegrees=20, twistLimitDegrees=10}
 {bone1Name=dyn_flapMid_r_3, bone2Name=dyn_flapMid_r_2, swing1LimitDegrees=20, swing2LimitDegrees=20, twistLimitDegrees=10}
 {
  bone1Name=dyn_flapMid4
  bone2Name=dyn_flapMid_r_3
  swing1LimitDegrees=45
  swing1Motion=Locked
  swing2LimitDegrees=45
  swing2Motion=Locked
  twistLimitDegrees=45
  twistMotion=Locked
 }
 {bone1Name=dyn_BaseFlap05_r, bone2Name=dyn_BaseFlap04, swing1LimitDegrees=45, swing2LimitDegrees=45, twistLimitDegrees=10}
 {bone1Name=dyn_BaseFlap05_l, bone2Name=dyn_BaseFlap04, swing1LimitDegrees=45, swing2LimitDegrees=45, twistLimitDegrees=10}
]
